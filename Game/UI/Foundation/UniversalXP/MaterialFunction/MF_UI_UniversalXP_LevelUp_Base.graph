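# Material function that builds the base look of the Universal XP level-up UI icon, outputting Color, Opacity and Alpha Mask.
# One packed icon texture supplies outline, backing, gradient and alpha masks in its RGBA channels.
# Centered and offset radial gradients on 9-slice UI UVs blend five colour parameters into procedural shading.
# Nested lerps composite the colour layers, while Opacity is a remapped, saturated version of the Alpha Mask output.
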MaterialFunction /Game/UI/Foundation/UniversalXP/MaterialFunction/MF_UI_UniversalXP_LevelUp_Base
graph {nodes [
  n1 = VectorParameter() {parameterName=Colour5, defaultValue=<rgba 0.0123 0.0212 0.0395 1>}
  n2 = NamedRerouteDeclaration(None=VectorParameter() {parameterName=Colour4, defaultValue=<rgba 0.011 0.0545 0.2423 1>}.RGB) {name=Colour4, nodeColor=<rgba 0 0 0 1>}
  n3 = LinearInterpolate(Alpha=NamedRerouteUsage()) {constA=0.3, constB=0}
  n4 = NamedRerouteUsage()
  n5 = NamedRerouteUsage()
  n6 = Add(A=Multiply(A=@n4, B=NamedRerouteUsage()), B=Multiply(A=@n5, B=NamedRerouteUsage()))
  n7 = LinearInterpolate(A=@n4, Alpha=@n5)
  n8 = LinearInterpolate(
   A=LinearInterpolate(A=NamedRerouteUsage(), B=NamedRerouteDeclaration(None=@n1.RGB) {name=Colour5, nodeColor=<rgba 0 0 0 1>}, Alpha=NamedRerouteUsage())
   B=LinearInterpolate(A=Multiply(A=@n2, B=@n3), B=Saturate(None=@n6), Alpha=Multiply(A=@n7) {constB=0.125})
   Alpha=NamedRerouteUsage()
  )
  n9 = NamedRerouteUsage()
  n10 = NamedRerouteUsage()
  FunctionOutput(None=LinearInterpolate(A=LinearInterpolate(A=@n8, B=@n9, Alpha=@n10), Alpha=NamedRerouteUsage())) {outputName=Color}
  n11 = NamedRerouteUsage()
  n12 = Constant() {r=0.4}
  n13 = Constant() {r=1}
  FunctionOutput(None=Saturate(None=MaterialFunctionCall(Input=@n11, Input Low=@n12, Input High=@n13) {materialFunction=&RemapValue_0-1}.Result)) {outputName=Opacity, sortPriority=2}
  FunctionOutput(None=@n11) {outputName=Alpha Mask, sortPriority=1}
  n14 = TextureObjectParameter() {
   parameterName=IconTexture
   samplerType=SAMPLERTYPE_LinearColor
   texture=/Game/UI/Foundation/UniversalXP/UniversalXP_LevelUp/T_UI_Icon_UniversalXP_LevelUp
  }
  n15 = TextureSample(Tex=@n14)
  NamedRerouteDeclaration(None=@n15.R) {name=OutlineMask, nodeColor=<rgba 0 0 0 1>}
  NamedRerouteDeclaration(None=@n15.G) {name=BackingMask, nodeColor=<rgba 0 0 0 1>}
  NamedRerouteDeclaration(None=MaterialFunctionCall(UV=MaterialFunctionCall() {materialFunction=&GetUserInterfaceUV}.9-Slice UV) {materialFunction=&MF_UI_RadialGradient}.Smooth) {name=Centered Radial Gradient, nodeColor=<rgba 0 0 0 1>}
  NamedRerouteDeclaration(None=@n15.B) {name=GradientMask, nodeColor=<rgba 0 0 0 1>}
  NamedRerouteDeclaration(None=@n15.A) {name=Alpha, nodeColor=<rgba 0 0 0 1>}
  n16 = NamedRerouteDeclaration(None=VectorParameter() {parameterName=Colour3, defaultValue=<rgba 0.0185 0.0908 0.5029 1>}.RGB) {name=Colour3, nodeColor=<rgba 0 0 0 1>}
  n17 = MaterialFunctionCall() {materialFunction=&GetUserInterfaceUV}
  n18 = Constant() {r=0.25}
  n19 = NamedRerouteDeclaration(None=MaterialFunctionCall(UV=Add(A=@n17.9-Slice UV, B=@n18)) {materialFunction=&MF_UI_RadialGradient}.Smooth) {name=TopRadialMask, nodeColor=<rgba 0 0 0 1>}
  n20 = VectorParameter() {parameterName=Colour1, defaultValue=<rgba 0.1812 0.956 0.956 1>}
  n21 = VectorParameter() {parameterName=Colour2, defaultValue=<rgba 0.0802 0.4072 0.9734 1>}
  n22 = Multiply(A=@n18) {constB=-1}
  n23 = NamedRerouteDeclaration(None=MaterialFunctionCall(UV=Add(A=@n17.9-Slice UV, B=@n22)) {materialFunction=&MF_UI_RadialGradient}.Smooth) {name=BottomRadialMask, nodeColor=<rgba 0 0 0 1>}
  n24 = Add(
   A=LinearInterpolate(A=@n16, B=NamedRerouteDeclaration(None=@n20.RGB) {name=Colour1, nodeColor=<rgba 0 0 0 1>}, Alpha=@n19)
   B=LinearInterpolate(A=@n16, B=NamedRerouteDeclaration(None=@n21.RGB) {name=Colour2, nodeColor=<rgba 0 0 0 1>}, Alpha=@n23)
  )
  NamedRerouteDeclaration(None=LinearInterpolate(A=NamedRerouteUsage(), B=Saturate(None=@n24), Alpha=LinearInterpolate(A=@n19, Alpha=@n23))) {name=Gradient, nodeColor=<rgba 0 0 0 1>}
 ]}
